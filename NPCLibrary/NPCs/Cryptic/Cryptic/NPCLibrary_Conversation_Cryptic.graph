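# Conversation database for the NPC Cryptic, a relic collector who greets the player with flavor lines.
# The Cryptic entry point leads to one speech node with a default line and prioritized contextual alternatives.
# A first-ever-conversation requirement gates the introduction, while the other alternatives have no requirements and use OutfitReaction priority rows.
# The speech node then branches through link nodes to the shared NPCLibrary Quests and BasicServices conversation entries.
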ConversationDatabase /NPCLibrary/NPCs/Cryptic/Cryptic/NPCLibrary_Conversation_Cryptic
entry_points [{tag=Conversation.Entry.Cryptic, destinations [#1]}]
nodes [
 ConversationEntryPointNode_7 {
  id=#1
  class=ConversationEntryPointNode
  outputs [#2]
  properties {entryTag {tagName=Conversation.Entry.Cryptic}}
 }
 Speech_C_4 {
  id=#2
  class=Speech_C
  outputs [#3, #4]
  properties {
   nodeName=""
   generalConfig {
    defaultMessage=You'd be shocked at the price of ancient relics looted from ransacked realities.
    contextualMessages [
     {
      message=I'm a collector of relics and artifacts. Which are... expensive.
      contextRequirements [{requirement {class=FortControllerRequirement_IsFirstEverConversationWithNPC}}]
      priority {curve {rowName=Tandem.Speech.Priority.FirstEverInteraction}}
     }
     {
      message=You'd be shocked at the price of ancient relics looted from ransacked realities.
      requirementMatchPolicy=RequireAny
      contextRequirements []
      priority {curve {curveTable=null, rowName=Tandem.Speech.Priority.OutfitReaction}}
     }
     {
      message="I'm willing to get my hands dirty, for a price."
      requirementMatchPolicy=RequireAny
      contextRequirements []
      priority {curve {curveTable=null, rowName=Tandem.Speech.Priority.OutfitReaction}}
     }
     {
      message=I wouldn't mind a glance inside the Cold Blooded vaults.
      requirementMatchPolicy=RequireAny
      contextRequirements []
      priority {curve {curveTable=null, rowName=Tandem.Speech.Priority.OutfitReaction}}
     }
    ]
   }
  }
 }
 ConversationLinkNode_0 {id=#4, class=ConversationLinkNode, properties {remoteEntryTag {tagName=Conversation.Entry.BasicServices}}}
 ConversationLinkNode_1 {id=#3, class=ConversationLinkNode, properties {remoteEntryTag {tagName=Conversation.Entry.NPCLibrary.Quests}}}
]
